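# Physics asset for the F_MED_Binary backpack cosmetic, giving capsule bodies to the pack and its two winged assemblies.
# Each side hangs a wing master body rigidly locked to the pack, followed by a five-link wing chain ending in two flaps on the last segment.
# Joints use very tight angular limits of a few degrees, so wings and flaps can wobble only slightly.
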PhysicsAsset /BRCosmetics/Accessories/FORT_Backpacks/Backpack_F_MED_Binary/Meshes/F_MED_Binary_Pack_Physics
bodies {
 dyn_wingFlap_01_l [{
   shapeType=Capsule
   length=9.6784
   radius=3.2422
   shapeName=dyn_wingFlap_01_l_capsule
   center=<6.9164 0.0633 4.4564>
   rotation=<rot 0.8661 -90 42.1543>
  }]
 dyn_wingFlap_01_r [{
   shapeType=Capsule
   length=9.6784
   radius=3.2422
   shapeName=dyn_wingFlap_01_r_capsule
   center=<6.9165 0.0628 -4.4563>
   rotation=<rot -0.8642 -90 -42.1547>
  }]
 dyn_wingFlap_02_l [{
   shapeType=Capsule
   length=14.0698
   radius=4.0843
   shapeName=dyn_wingFlap_02_l_capsule
   center=<8.8875 -0.4752 6.3114>
   rotation=<rot 0 86.9826 -45.1744>
  }]
 dyn_wingFlap_02_r [{
   shapeType=Capsule
   length=14.0698
   radius=4.0842
   shapeName=dyn_wingFlap_02_r_capsule
   center=<8.8876 -0.4738 -6.3114>
   rotation=<rot 0 -93.0182 -45.174>
  }]
 dyn_wing_01_l [{
   shapeType=Capsule
   length=4.7194
   radius=4.2679
   shapeName=dyn_wing_01_l_capsule
   center=<3.7867 0.0006 0.3282>
   rotation=<rot 0 89.4276 -75.2092>
  }]
 dyn_wing_01_l_MSTR [{shapeType=Capsule, length=0.505, radius=0.505, shapeName=dyn_wing_01_l_MSTR_capsule, rotation=<rot 0 -90 0>}]
 dyn_wing_01_r [{
   shapeType=Capsule
   length=4.7194
   radius=4.2679
   shapeName=dyn_wing_01_r_capsule
   center=<3.7867 0.0006 -0.3282>
   rotation=<rot 0 -90.5724 -75.2092>
  }]
 dyn_wing_01_r_MSTR [{shapeType=Capsule, length=0.505, radius=0.505, shapeName=dyn_wing_01_r_MSTR_capsule, rotation=<rot 0 -90 0>}]
 dyn_wing_02_l [{
   shapeType=Capsule
   length=9.7358
   radius=4.836
   shapeName=dyn_wing_02_l_capsule
   center=<4.0499 0.0014 0.2988>
   rotation=<rot 0.2749 -90 24.163>
  }]
 dyn_wing_02_r [{
   shapeType=Capsule
   length=9.7358
   radius=4.836
   shapeName=dyn_wing_02_r_capsule
   center=<4.0498 0.0014 -0.2988>
   rotation=<rot -0.2751 -90 -24.163>
  }]
 dyn_wing_03_l [{
   shapeType=Capsule
   length=3.9652
   radius=2.3039
   shapeName=dyn_wing_03_l_capsule
   center=<1.7976 0.014 -0.5829>
   rotation=<rot 0 -89.5477 -66.4394>
  }]
 dyn_wing_03_r [{
   shapeType=Capsule
   length=3.9652
   radius=2.3039
   shapeName=dyn_wing_03_r_capsule
   center=<1.7976 0.014 0.5829>
   rotation=<rot 0 90.4523 -66.4394>
  }]
 dyn_wing_04_l [{shapeType=Capsule, length=0.505, radius=0.505, shapeName=dyn_wing_04_l_capsule, rotation=<rot 0 -90 0>}]
 dyn_wing_04_r [{shapeType=Capsule, length=0.505, radius=0.505, shapeName=dyn_wing_04_r_capsule, rotation=<rot 0 -90 0>}]
 dyn_wing_05_l [{
   shapeType=Capsule
   length=33.8864
   radius=15.1786
   shapeName=dyn_wing_05_l_capsule
   center=<27.4945 0.0477 5.4345>
   rotation=<rot 0 90.099 -84.3386>
  }]
 dyn_wing_05_r [{
   shapeType=Capsule
   length=33.8864
   radius=15.1786
   shapeName=dyn_wing_05_r_capsule
   center=<27.4945 0.0476 -5.4345>
   rotation=<rot 0 -89.9012 -84.3387>
  }]
 pack [{shapeType=Capsule, length=13.995, radius=12.5555, shapeName=pack_capsule}]
}
constraints [
 {
  bone1Name=dyn_wing_01_l_MSTR
  bone2Name=pack
  swing1LimitDegrees=45
  swing1Motion=Locked
  swing2LimitDegrees=45
  swing2Motion=Locked
  twistLimitDegrees=45
  twistMotion=Locked
 }
 {bone1Name=dyn_wing_01_l, bone2Name=dyn_wing_01_l_MSTR, swing1LimitDegrees=1, swing2LimitDegrees=3, twistLimitDegrees=1}
 {bone1Name=dyn_wing_02_l, bone2Name=dyn_wing_01_l, swing1LimitDegrees=3, swing2LimitDegrees=1, twistLimitDegrees=1}
 {bone1Name=dyn_wing_03_l, bone2Name=dyn_wing_02_l, swing1LimitDegrees=3, swing2LimitDegrees=3, twistLimitDegrees=1}
 {bone1Name=dyn_wing_04_l, bone2Name=dyn_wing_03_l, swing1LimitDegrees=3, swing2LimitDegrees=3, twistLimitDegrees=3}
 {bone1Name=dyn_wing_05_l, bone2Name=dyn_wing_04_l, swing1LimitDegrees=1, swing2LimitDegrees=3, twistLimitDegrees=1}
 {bone1Name=dyn_wingFlap_01_l, bone2Name=dyn_wing_05_l, swing1LimitDegrees=5, swing2LimitDegrees=1, twistLimitDegrees=3}
 {bone1Name=dyn_wingFlap_02_l, bone2Name=dyn_wing_05_l, swing1LimitDegrees=5, swing2LimitDegrees=1, twistLimitDegrees=3}
 {
  bone1Name=dyn_wing_01_r_MSTR
  bone2Name=pack
  swing1LimitDegrees=45
  swing1Motion=Locked
  swing2LimitDegrees=45
  swing2Motion=Locked
  twistLimitDegrees=45
  twistMotion=Locked
 }
 {bone1Name=dyn_wing_01_r, bone2Name=dyn_wing_01_r_MSTR, swing1LimitDegrees=1, swing2LimitDegrees=3, twistLimitDegrees=1}
 {bone1Name=dyn_wing_02_r, bone2Name=dyn_wing_01_r, swing1LimitDegrees=3, swing2LimitDegrees=1, twistLimitDegrees=1}
 {bone1Name=dyn_wing_03_r, bone2Name=dyn_wing_02_r, swing1LimitDegrees=3, swing2LimitDegrees=3, twistLimitDegrees=1}
 {bone1Name=dyn_wing_04_r, bone2Name=dyn_wing_03_r, swing1LimitDegrees=3, swing2LimitDegrees=3, twistLimitDegrees=3}
 {bone1Name=dyn_wing_05_r, bone2Name=dyn_wing_04_r, swing1LimitDegrees=1, swing2LimitDegrees=3, twistLimitDegrees=1}
 {bone1Name=dyn_wingFlap_01_r, bone2Name=dyn_wing_05_r, swing1LimitDegrees=3, swing2LimitDegrees=1, twistLimitDegrees=3}
 {bone1Name=dyn_wingFlap_02_r, bone2Name=dyn_wing_05_r, swing1LimitDegrees=3, swing2LimitDegrees=1, twistLimitDegrees=3}
]
